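; Katana melee primary action gameplay ability for Fortnite player pawns, gated so it cannot start during the in-air attack leap.
; CanActivateAbility casts the avatar to FortPlayerPawn and allows activation only when it lacks the Granted.Athena.Katana.InAirAttackLeap tag; a failed cast blocks it.
; Activation defers to the parent class's logic and then immediately calls EndAbility.
; The end-ability event only forwards to the parent.
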(event Ability|EventActivateAbility
  (|CallParentFunction)
  (Ability|EndAbility))

(event Ability|EventOnEndAbility
  (params (bWasCancelled :type Boolean))
  (|CallParentFunction bWasCancelled))

(fn CanActivateAbility
  (params (ActorInfo :type Gameplay%20Ability%20Actor%20Info%20Structure) (Handle :type Gameplay%20Ability%20Spec%20Handle%20Structure))
  (bind (_owneractor _avataractor) (Utilities|Struct|BreakGameplayAbilityActorInfo ActorInfo))
  (bind _asfort_player_pawn (Utilities|Casting|CastToFortPlayerPawn _avataractor)
    (:then
      (return 0 (not (GameplayTags|HasMatchingGameplayTag _asfort_player_pawn "(TagName=\"Granted.Athena.Katana.InAirAttackLeap\")"))))
    (:CastFailed
      (return 0 false))))
