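; Katana Mantis technique ability that runs the dash charge, charge eject and dash attack, reacting to technique start, hit and end events.
; Charging lowers gravity and curves time dilation until a timer forces the dash; ejecting launches sliding or sprinting players upward.
; The attack drives dash root motion from technique metadata, adjusts targeting and ignores hit actors' collision.
; Tag listeners end it on swimming, grinding, DBNO, skydiving, teleporting or death.
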(event Ability|EventActivateAbility
  (bind _fortplayerpawn (Variables|Default|GetFortPlayerPawn))
  (|CallParentFunction)
  (bind _asfort_player_pawn (Utilities|Casting|CastToFortPlayerPawn (Ability|GetActivatingPawn)))
  (Variables|Default|SetFortPlayerPawn _asfort_player_pawn)
  (bind _returnvalue (Ability|ApplyGameplayEffectToOwner "/KatanaGameplay/Items/Katana/Abilities/DashAttack/GE_Katana_DashAttack_SpeedBoost.GE_Katana_DashAttack_SpeedBoost_C"))
  (Variables|Default|SetSpeedBoostEffectHandle _returnvalue)
  (Character|UnCrouch _fortplayerpawn)
  (bind _asfort_movement_comp_character (Utilities|Casting|CastToFortMovementComp_Character (Variables|/Script/Engine.Character|GetCharacterMovement _fortplayerpawn)))
  (NetTolerance|SetUseHighAccelerationTolerance _asfort_movement_comp_character true "Katana")
  (Ability|GameplayEffect|AddLooseGameplayTags (Variables|Default|GetFortPlayerPawn) (Variables|Dash|Charge|GetBlockEquipAnimTag)))

(event Ability|EventOnEndAbility
  (params (bWasCancelled :type Boolean))
  ;[1 0.4649 0.09567] End Ability
  (|CallParentFunction bWasCancelled)
  (Utilities|FlowControl|Sequence
    (:then_0
      (CallFunction|ClearMoveIgnoredActors)
      (CallFunction|StopDashAttackRootMotion)
      (CallFunction|StopDashCharge)
      (Utilities|FlowControl|Sequence
        (:then_0
          (if (Variables|Default|GetbHasSlideSprintBlockTags)
            (Ability|GameplayEffect|RemoveLooseGameplayTags (Variables|Default|GetFortPlayerPawn) (Variables|Default|GetBlockSlideSprintTags))))
        (:then_1
          (if (GameplayTags|HasAnyMatchingGameplayTags (Variables|Default|GetFortPlayerPawn) (Variables|Dash|Charge|GetBlockEquipAnimTag))
            (Ability|GameplayEffect|RemoveLooseGameplayTags (Variables|Default|GetFortPlayerPawn) (Variables|Dash|Charge|GetBlockEquipAnimTag))))))
    (:then_1
      (bind _fortplayerpawn (Variables|Default|GetFortPlayerPawn)
        (:then
          (Animation|AnimTrailsNotify _fortplayerpawn false))
        (:else
          _))
      (Ability|RemoveGameplayEffectFromOwnerWithAssetTags "(GameplayTags=((TagName=\"Granted.KatanaGameplay.Weapon.Melee.Katana.Chargeup\")))")
      (Ability|RemoveGameplayEffectFromOwnerWithHandle (Variables|Default|GetSpeedBoostEffectHandle))
      (Utilities|FlowControl|Sequence
        (:then_0
          (bind _waitswimasynctask (Variables|Default|GetWaitSwimAsyncTask)
            (:then
              (GameplayTasks|EndTask _waitswimasynctask))
            (:else)))
        (:then_1
          (bind _waitgrindingasynctask (Variables|Default|GetWaitGrindingAsyncTask)
            (:then
              (GameplayTasks|EndTask _waitgrindingasynctask))
            (:else)))
        (:then_2
          (bind _waitdbnoasynctask (Variables|Default|GetWaitDBNOAsyncTask)
            (:then
              (GameplayTasks|EndTask _waitdbnoasynctask))
            (:else)))
        (:then_3
          (bind _fortplayerpawn_2 (Variables|Default|GetFortPlayerPawn)
            (:then
              (EventDispatchers|UnbindallEventsfromOnDied _fortplayerpawn_2))
            (:else)))
        (:then_4
          (bind _waitskydivingasynctask (Variables|Default|GetWaitSkydivingAsyncTask)
            (:then
              (GameplayTasks|EndTask _waitskydivingasynctask))
            (:else)))
        (:then_5
          (bind _waitteleportedasynctask (Variables|Default|GetWaitTeleportedAsyncTask)
            (:then
              (GameplayTasks|EndTask _waitteleportedasynctask))
            (:else)))))
    (:then_2
      (bind _dashchargeejectconstantforce (Variables|Dash|Eject|GetDashChargeEjectConstantForce)
        (:then
          (GameplayTasks|EndTask _dashchargeejectconstantforce))
        (:else)))
    (:then_3
      (bind _fortplayerpawn_3 (Variables|Default|GetFortPlayerPawn)
        (:then
          (bind _asfort_movement_comp_character (Utilities|Casting|CastToFortMovementComp_Character (Variables|/Script/Engine.Character|GetCharacterMovement _fortplayerpawn_3)))
          (NetTolerance|SetUseHighAccelerationTolerance _asfort_movement_comp_character false "Katana"))
        (:else)))
    (:then_4
      (switch name (Variables|Default|GetCurrentTechnique)
        (:DashAttack
          (CallFunction|ApplyFallDamageImmunity (Variables|Default|GetGE_EndAbility_FallDamageImmunity)))))))

(event Custom|OnDashAttackStarted
  ;[0.08689 0.4707 1] Dash Attack
  (bind _fortplayerpawn (Variables|Default|GetFortPlayerPawn)
    (:then
      (Networking|ForceNetUpdate _fortplayerpawn)
      (if (Mantis|HasCurrentTechniqueMetadataBeenSet (Actor|GetComponentbyClass _fortplayerpawn "/Script/MantisRuntime.FortMantisPawnComponent"))
        (CallFunction|CacheDashAttackMetadata)
        (CallFunction|StartDashAttackRootMotion)
        (if (>= (ScalableFloat|GetValueatLevel (Variables|Default|GetForceRemoveBalloonsOnDash)) 1.0)
          (Ability|SendGameplayEvent (Variables|Default|GetEventTag_RemoveBalloons)))
        (else
          (Mantis|BindEventtoOnTechniqueMetadataReady (Actor|GetComponentbyClass _fortplayerpawn "/Script/MantisRuntime.FortMantisPawnComponent") (delegate Custom|OnDashAttackMetadataReady)))))
    (:else)))

(event Custom|OnDashChargeEjectStarted
  (bind _fortplayerpawn (Variables|Default|GetFortPlayerPawn)
    (:then
      (Networking|ForceNetUpdate _fortplayerpawn)
      (if (GameplayTags|HasMatchingGameplayTag (Variables|Default|GetFortPlayerPawn) "(TagName=\"MovementMode.Sliding\")")
        ;[1 1 1] Sliding
        ; Note: Cached Eject Velocity is zero-ed out due to movement exploit FORT-1104642
        (Variables|Dash|Eject|SetCachedEjectVelocity "0, 0, 0")
        (Character|UnCrouch (Variables|Default|GetFortPlayerPawn))
        (Pawn|Components|CharacterMovement|SetMovementMode (Variables|/Script/Engine.Character|GetCharacterMovement (Variables|Default|GetFortPlayerPawn)) "MOVE_Falling" 0)
        (CallFunction|LaunchPlayerUp (ScalableFloat|GetValueatLevel (Variables|Dash|Eject|GetJumpOutOfSlidingZVelocity)))
        (elif (GameplayTags|HasMatchingGameplayTag (Variables|Default|GetFortPlayerPawn) "(TagName=\"Granted.Athena.Ability.TomatoSprinkle.Sprinting\")")
          ;[1 1 1] Sprinting
          (Variables|Dash|Eject|SetCachedEjectVelocity (Pawn|Components|CharacterMovement|GetLastUpdateVelocity (Variables|/Script/Engine.Character|GetCharacterMovement (Variables|Default|GetFortPlayerPawn))))
          (CallFunction|LaunchPlayerUp (ScalableFloat|GetValueatLevel (Variables|Dash|Eject|GetJumpOutOfSlidingZVelocity)))
          (else
            ;[1 1 1] TODO: Plug this back in as a failsafe.
            ; If the player has any of the given tags, don't stop the dash charge eject technique, the DashChargeGA will stop the technique after launching the player.
            (if (not (GameplayTags|HasAnyMatchingGameplayTags (Variables|Default|GetFortPlayerPawn) (Variables|Dash|Eject|GetEjectedByGATags)))
              (CallFunction|StopDashChargeEject))))))
    (:else)))

(event Custom|OnDashAttackMetadataReady
  (params (TechniqueMetadata :type Fort%20Mantis%20Technique%20Metadata%20Structure%20%28by%20ref%29))
  ;[0.08689 0.4707 1] Dash Attack
  (bind _fortplayerpawn (Variables|Default|GetFortPlayerPawn)
    (:then
      (Mantis|UnbindallEventsfromOnTechniqueMetadataReady (Actor|GetComponentbyClass _fortplayerpawn "/Script/MantisRuntime.FortMantisPawnComponent"))
      (CallFunction|CacheDashAttackMetadata)
      (CallFunction|StartDashAttackRootMotion)
      (if (>= (ScalableFloat|GetValueatLevel (Variables|Default|GetForceRemoveBalloonsOnDash)) 1.0)
        (Ability|SendGameplayEvent (Variables|Default|GetEventTag_RemoveBalloons))))
    (:else)))

(event Custom|OnDashChargeStarted
  ;[1 0.2487 0.8122] Dash Charge
  (Utilities|FlowControl|Sequence
    (:then_0
      ;;[0.2929 0.4697 1] Remove the BlockEquipAnim tag (potentially added from the charge eject GA)
      (Ability|HasAuthority
        (:False
          (Ability|GameplayEffect|RemoveLooseGameplayTags (Variables|Default|GetFortPlayerPawn) (Variables|Dash|Charge|GetBlockEquipAnimTag))
          (Ability|GameplayEffect|AddLooseGameplayTags (Variables|Default|GetFortPlayerPawn) (Variables|Default|GetBlockSlideSprintTags))
          (Variables|Default|SetbHasSlideSprintBlockTags true))))
    (:then_1
      ;;[0.06569 1 0.6531] Camera Modifier
      (bind _fortplayerpawn (Variables|Default|GetFortPlayerPawn)
        (:then
          (bind _outmodifier (Class|WeaponFunctionLibrary|FindorAddCameraModifier "/KatanaGameplay/Items/Katana/CameraModifiers/B_Katana_DashCharge_CameraModifier.B_Katana_DashCharge_CameraModifier_C" (Variables|/Script/FortniteGame.FortPawn|GetCurrentWeapon _fortplayerpawn)))
          (bind _output_get (Variables|Dash|Charge|SetChargeCameraModifier _outmodifier))
          (CameraModifier|EnableModifier _output_get))
        (:else)))
    (:then_2
      ;;[0.2102 0.4569 1] Notify Dash Charge
      (bind _fortplayerpawn_2 (Variables|Default|GetFortPlayerPawn)
        (:then
          (bind _asb_melee_katana (Utilities|Casting|CastToB_Melee_Katana (Weapon|GetCurrentWeapon _fortplayerpawn_2)))
          (Class|BMeleeKatana|SetDashChargeEnabled _asb_melee_katana true)
          ;;[1 1 1] Reset fall height to pevent fall dmg
          (Pawn|Character|ResetFallingHeight (Variables|Default|GetFortPlayerPawn)))
        (:else)))
    (:then_3
      ;;[1 0.2027 0.5869] Time Dilation as slomo - applied to owning client and server here, applied to remote clients in the Charge GCNL.
      (bind _fortplayerpawn_3 (Variables|Default|GetFortPlayerPawn)
        (:then
          (Networking|ForceNetUpdate _fortplayerpawn_3)
          (Variables|Dash|Charge|SetChargeSlomoStartTime (Utilities|Time|GetTimeSeconds))
          (bind _returnvalue (Utilities|Time|SetTimerbyEvent (EventDispatchers|CreateEvent :!event Custom|UpdateSlomo) 0.05 true))
          (Variables|Dash|Charge|SetUpdateSlomoTimerHandle _returnvalue)
          (bind _fortplayerpawn_4 (Variables|Default|GetFortPlayerPawn)
            (:then
              (Character|StopJumping _fortplayerpawn_4)
              (bind (_velocity_x _velocity_y _velocity_z) (Math|Vector|BreakVector (Variables|/Script/Engine.MovementComponent|GetVelocity_X (Variables|/Script/Engine.Character|GetCharacterMovement _fortplayerpawn_4))))
              (Variables|/Script/Engine.MovementComponent|SetVelocity_X (Math|Vector|MakeVector _velocity_x _velocity_y (* _velocity_z (ScalableFloat|GetValueatLevel (Variables|Dash|Charge|GetChargeZVelocityMult)))) (Variables|/Script/Engine.Character|GetCharacterMovement _fortplayerpawn_4))
              (Variables|/Script/Engine.CharacterMovementComponent|SetGravityScale (ScalableFloat|GetValueatLevel (Variables|Dash|Charge|GetChargeGravityScaleMult)) (Variables|/Script/Engine.Character|GetCharacterMovement (Variables|Default|GetFortPlayerPawn))) ; Different value for gravity.\nThis lets us maintain momentum in XY\nbut fall slower.
            )
            (:else)))
        (:else)))
    (:then_4
      ;;[0.0389 0.9303 1] Force Dash
      (bind _returnvalue_2 (Utilities|Time|SetTimerbyEvent (EventDispatchers|CreateEvent :!event Custom|ForceDash) (ScalableFloat|GetValueatLevel (Variables|Dash|Charge|GetMaxDashChargeTime))))
      (Variables|Dash|Charge|SetForceDashTimerHandle _returnvalue_2))))

(event Mantis|EventOnMantisPreTechniqueSelection
  ; FORT-1057131 - Give visual/audio feedback when Dash is out of ammo/charges for Katana and trying to use Dash
  (if (ScalableFloat|GetValueasBool (Variables|Default|GetHF_DashAttackNoAmmoNotifcationEnabled))
    (Utilities|FlowControl|DelayUntilNextTick) ; Allow for Current Technique to Update
    (if (and (FortWeapon|Ammo|IsOutOfAmmo (Weapon|GetCurrentWeapon (Variables|Default|GetFortPlayerPawn))) (Utilities|String|Contains (Utilities|String|ToString(Name) (Variables|Default|GetCurrentTechnique)) "Dash"))
      (bind _returnvalue (Utilities|Time|GetGameTimeinSeconds))
      (if (> (- _returnvalue (Variables|Default|GetLastPlayedAbilityCooldownGameTime)) (Variables|Default|GetPlayAbilityOnCooldownSoundInterval))
        (Variables|Default|SetLastPlayedAbilityCooldownGameTime _returnvalue)
        (Audio|PlaySound2D (Variables|Default|GetAbilityOnCooldownSound))
        (FailedAction|UnabletoPerformAction (Variables|Default|GetFortPlayerPawn) true (Variables|Default|GetFailed%20Reason))))))

(event Custom|CacheDashAttackMetadata
  ;[0.08689 0.4707 1] Dash Attack
  (bind _fortplayerpawn (Variables|Default|GetFortPlayerPawn)
    (:then
      (Networking|ForceNetUpdate _fortplayerpawn)
      (Utilities|FlowControl|Sequence
        (:then_0
          (bind (_returnvalue_intvalues _returnvalue_floatvalues _returnvalue_vectorvalues) (Utilities|Struct|BreakFortMantisTechniqueMetadata (Mantis|GetCurrentTechniqueMetadata (Actor|GetComponentbyClass _fortplayerpawn "/Script/MantisRuntime.FortMantisPawnComponent"))))
          (if (.ReturnValue (Utilities|Map|Find _returnvalue_intvalues "(TagName=\"Mantis.Metadata.Katana.ShouldStopAfterDash\")"))
            (Variables|Dash|Attack|SetbShouldStopAfterDash (Math|Conversions|ToBoolean(Integer) (Utilities|Map|Find _returnvalue_intvalues "(TagName=\"Mantis.Metadata.Katana.ShouldStopAfterDash\")")))))
        (:then_1
          (if (.ReturnValue (Utilities|Map|Find _returnvalue_floatvalues "(TagName=\"Mantis.Metadata.Katana.DashDistance\")"))
            (Variables|Dash|Attack|SetActualDashDistance (Utilities|Map|Find _returnvalue_floatvalues "(TagName=\"Mantis.Metadata.Katana.DashDistance\")"))))
        (:then_2
          (if (.ReturnValue (Utilities|Map|Find _returnvalue_vectorvalues "(TagName=\"Mantis.Metadata.Katana.DashDirection\")"))
            (Variables|Dash|Attack|SetDashDirection (Math|Conversions|RotationFromXVector (Utilities|Map|Find _returnvalue_vectorvalues "(TagName=\"Mantis.Metadata.Katana.DashDirection\")")))))
        (:then_3
          (if (.ReturnValue (Utilities|Map|Find _returnvalue_vectorvalues "(TagName=\"Mantis.Metadata.Katana.DashOrigin\")"))
            (Variables|Dash|Attack|SetDashOrigin (Utilities|Map|Find _returnvalue_vectorvalues "(TagName=\"Mantis.Metadata.Katana.DashOrigin\")"))))
        (:then_4
          (CallFunction|ValidateDashMetadata))))
    (:else)))

(event Mantis|EventOnMantisAblilityTechniqueEnded
  (params (SequenceName :type Name :default "None"))
  (|CallParentFunction SequenceName)
  (switch name SequenceName
    (:InAirAttack_Leap
      (if (ScalableFloat|GetValueasBool (Variables|Default|GetHF_AirLeapBlocksPrimaryAttack))
        (Ability|GameplayEffect|RemoveLooseGameplayTags (Variables|Default|GetFortPlayerPawn) (Variables|Default|GetBlockPrimaryAttackTag))))))

(event Mantis|EventOnMantisInputEvent
  (params (InputType :type EFortMantisTechniqueActivationInputType%20Enum) (InputTiming :type EFortMantisTechniqueActivationTimingType%20Enum) (bDidInputStartTechnique :type Boolean) (bDidInputQueueTechnique :type Boolean) (InputHeldDuration :type Float%20%28single-precision%29))
  ;[0.3532 0.2572 1] Cache the duration that the primary mantis input was held if it was released and started or queued a technique. That data will be sent to the technique metadata which will be replicated to the server and used to determine how far the player will dash.
  (if (Math|Boolean|ANDBoolean (Utilities|Enum|Equal(Enum) InputType "Secondary") (Utilities|Enum|Equal(Enum) InputTiming "Released") (or bDidInputStartTechnique bDidInputQueueTechnique))
    (Variables|Default|SetCurrentTechniqueInputHeldDuration InputHeldDuration)))

(event Custom|ListenForEndEvents
  ;[0.2118 0.4018 1] End this ability if the player enters certain states.
  ;;[1 0.815 0.06314] Delay until the next frame to prevent ending the technique the same frame it was started, which can cause issues for Mantis.
  (Utilities|FlowControl|DelayUntilNextTick)
  (Utilities|FlowControl|Sequence
    (:then_0
      (bind _asynctaskproxy (Ability|Tasks|WaitGameplayTagAdd "(TagName=\"MovementMode.SurfaceSwimming\")")
        (:then
          (Variables|Default|SetWaitSwimAsyncTask _asynctaskproxy))
        (:Added
          (if (!= (Variables|Default|GetCurrentTechnique) "DashChargeEject")
            (Ability|EndAbility)))))
    (:then_1
      (bind _asynctaskproxy_2 (Ability|Tasks|WaitGameplayTagAdd "(TagName=\"MovementMode.Grinding\")")
        (:then
          (Variables|Default|SetWaitGrindingAsyncTask _asynctaskproxy_2))
        (:Added
          (switch name (Variables|Default|GetCurrentTechnique)
            (:Default
              (Ability|EndAbility))))))
    (:then_2
      (bind _asynctaskproxy_3 (Ability|Tasks|WaitGameplayTagAdd "(TagName=\"Gameplay.Status.DBNO\")")
        (:then
          (Variables|Default|SetWaitDBNOAsyncTask _asynctaskproxy_3))
        (:Added
          (Ability|EndAbility))))
    (:then_3
      (bind _fortplayerpawn (Variables|Default|GetFortPlayerPawn)
        (:then
          (EventDispatchers|AssignOnDied _fortplayerpawn (delegate Custom|OnDied_Event)))
        (:else)))
    (:then_4
      (bind _asynctaskproxy_4 (Ability|Tasks|WaitGameplayTagAdd "(TagName=\"MovementMode.Skydiving\")")
        (:then
          (Variables|Default|SetWaitSkydivingAsyncTask _asynctaskproxy_4))
        (:Added
          (Ability|EndAbility))))
    (:then_5
      (bind _asynctaskproxy_5 (Ability|Tasks|WaitGameplayTagAdd "(TagName=\"Abilities.Creative.Teleporter.JustTeleported\")")
        (:then
          (Variables|Default|SetWaitTeleportedAsyncTask _asynctaskproxy_5))
        (:Added
          (Ability|EndAbility))))))

(event Mantis|EventOnMantisAblilityTechniqueStarted
  (params (OwningTags :type Gameplay%20Tag%20Container%20Structure) (SequenceName :type Name :default "None") (StartsSequence :type Boolean) (EndsSequence :type Boolean))
  ;[1 0.2238 0.8693] Ignore pawn collisions when dashing. Do this here instead of in the DashAttack GA for two reasons:
  ; - we already did a trace to deal damage to the players in the path of the dash, we can take advantage of that using the target data of the dash attack.
  ; - because the targeting for the damage and the application of the DashAttack GA happen at roughly the same time, the dash movement occasionally occurs before the targeting, resulting in the targeting failing because all pawns that would be dealt damage have their collision ignored.
  (|CallParentFunction OwningTags SequenceName StartsSequence EndsSequence)
  (Utilities|FlowControl|Sequence
    (:then_0
      (Variables|Default|SetCurrentTechnique SequenceName)
      (CallFunction|ClearMoveIgnoredActors)
      (Utilities|Time|ClearandInvalidateTimerbyHandle (Variables|Dash|Attack|GetEndDashAttackTimerHandle))
      (CallFunction|ListenforEndEvents))
    (:then_1
      (switch name SequenceName
        (:DashAttack
          (Variables|Default|SetbUseFinalSlashTargetingTransform false)
          (CallFunction|StopDashCharge)
          (CallFunction|OnDashAttackStarted)
          (Ability|HasAuthority
            (:False
              (CallFunction|AddDashAttackDynamicEffectContainer)
              (Ability|GameplayEffect|RemoveLooseGameplayTags (Variables|Default|GetFortPlayerPawn) (Variables|Default|GetBlockSlideSprintTags))
              (Variables|Default|SetbHasSlideSprintBlockTags false))))
        (:DashCharge
          (CallFunction|OnDashChargeStarted))
        (:DashChargeEject
          (CallFunction|OnDashChargeEjectStarted))
        (:InAirAttack_Leap
          (if (ScalableFloat|GetValueasBool (Variables|Default|GetHF_AirLeapBlocksPrimaryAttack))
            (Ability|GameplayEffect|AddLooseGameplayTags (Variables|Default|GetFortPlayerPawn) (Variables|Default|GetBlockPrimaryAttackTag))))))
    (:then_2
      (CallFunction|ResetDurabilityCost))))

(event Custom|StartDashAttackRootMotion
  ;[0.08689 0.4707 1] Dash Attack
  (bind _fortplayerpawn (Variables|Default|GetFortPlayerPawn)
    (:then
      (Networking|ForceNetUpdate _fortplayerpawn)
      (Variables|/Script/Engine.MovementComponent|SetVelocity "0, 0, 0" (Variables|/Script/Engine.Character|GetCharacterMovement _fortplayerpawn))
      (bind _asynctaskproxy (Ability|Tasks|ApplyRootMotion_DashForce "None" (+ (Variables|Dash|Attack|GetDashOrigin) (* (Math|Vector|GetForwardVector (Variables|Dash|Attack|GetDashDirection)) (Variables|Dash|Attack|GetActualDashDistance))) (ScalableFloat|GetValueatLevel (Variables|Dash|Attack|GetDashRootMotionDuration)) 6e+01 false "MOVE_Falling" false 0 "SetVelocity" (Math|Vector|SelectVector "0, 0, 0" (* (Math|Vector|GetForwardVector (Variables|Dash|Attack|GetDashDirection)) (ScalableFloat|GetValueatLevel (Variables|Dash|Attack|GetSpeedMaintanedAfterDash))) (Variables|Dash|Attack|GetbShouldStopAfterDash)) 0.0)
        (:then
          (Variables|Dash|Attack|SetDashAttackRootMotionTask _asynctaskproxy)
          (Utilities|FlowControl|Sequence
            (:then_0
              (CallFunction|ConsumeAmmo)
              (Camera|PushCameraMode "/KatanaGameplay/Items/Katana/CameraModifiers/B_Katana_DashAttack_CameraMode.B_Katana_DashAttack_CameraMode_C")
              (Utilities|FlowControl|Sequence
                (:then_0
                  (Utilities|FlowControl|Delay (/ (ScalableFloat|GetValueatLevel (Variables|Dash|Attack|GetDashRootMotionDuration)) 2.0))
                  (Variables|Default|SetbUseFinalSlashTargetingTransform true))
                (:then_1
                  (if (ScalableFloat|GetValueasBool (Variables|Default|GetEnablePostDashSlide))
                    (Ability|ApplyGameplayEffectToOwner "/KatanaGameplay/Items/Katana/Abilities/DashAttack/GE_Katana_PostDash_Slide_Grant.GE_Katana_PostDash_Slide_Grant_C")))))
            (:then_1
              (if (Class|AthenaFunctionLibrary|RowtoBool (Variables|Default|GetApplyCooldownToNyxGlass))
                (Ability|ApplyGameplayEffectToOwner "/KatanaGameplay/Items/Katana/Abilities/GE_KatanaToNyxGlassCooldown.GE_KatanaToNyxGlassCooldown_C")))))
        (:OnTimedOut
          _)
        (:OnTimedOutAndDestinationReached
          _))
      (CallFunction|StopDashAttackRootMotion))
    (:else)))

(event Custom|OnDied_Event
  (params (DamagedActor :type Actor%20Object%20Reference) (Damage :type Float%20%28single-precision%29) (InstigatedBy :type Controller%20Object%20Reference) (DamageCauser :type Actor%20Object%20Reference) (HitLocation :type Vector) (FHitComponent :type Primitive%20Component%20Object%20Reference) (BoneName :type Name) (Momentum :type Vector))
  ;[0.2118 0.4018 1] End this ability if the player enters certain states.
  (Ability|EndAbility))

(event Custom|UpdateSlomo
  (bind _dashchargetimedilationbuildupduration (Variables|Dash|Charge|GetDashChargeTimeDilationBuildupDuration))
  (bind _returnvalue (ScalableFloat|GetValueatLevel _dashchargetimedilationbuildupduration))
  (bind _returnvalue_2 (Utilities|Time|GetTimeSeconds))
  (bind _fallingdashchargetimedilation (Variables|Dash|Charge|GetFallingDashChargeTimeDilation))
  (bind _returnvalue_3 (ScalableFloat|GetValueatLevel _fallingdashchargetimedilation))
  (bind _fortplayerpawn (Variables|Default|GetFortPlayerPawn))
  (bind _returnvalue_4 (GameplayTags|HasMatchingGameplayTag _fortplayerpawn "(TagName=\"MovementMode.Falling\")"))
  (bind _grounddashchargetimedilation (Variables|Dash|Charge|GetGroundDashChargeTimeDilation))
  (bind _returnvalue_5 (ScalableFloat|GetValueatLevel _grounddashchargetimedilation))
  (bind _returnvalue_6 (Math|Float|SelectFloat _returnvalue_3 _returnvalue_5 _returnvalue_4))
  (bind _chargeslomostarttime (Variables|Dash|Charge|GetChargeSlomoStartTime))
  (bind _returnvalue_7 (- _returnvalue_2 _chargeslomostarttime))
  (bind _output (Utilities|RemapFloat _returnvalue_7 (Math|Vector2D|MakeVector2D 0.0 _returnvalue) "(X=0.000000,Y=1.000000)" true))
  ;[1 0.2487 0.8122] Dash Charge
  ;;[1 0.2622 0.6378] Time Dilation
  (if (>= _output 1.0)
    (Utilities|Time|ClearandInvalidateTimerbyHandle (Variables|Dash|Charge|GetUpdateSlomoTimerHandle))
    (else
      (bind _chargeslomocurve (Variables|Dash|Charge|GetChargeSlomoCurve)
        (:then
          (bind _fortplayerpawn_2 (Variables|Default|GetFortPlayerPawn)
            (:then
              (Variables|/Script/Engine.Actor|SetCustomTimeDilation (Utilities|RemapFloat (Math|Curves|GetFloatValue _chargeslomocurve _output) "(X=0.000000,Y=1.000000)" (Math|Vector2D|MakeVector2D _returnvalue_6 1.0) true) _fortplayerpawn_2)
              (bind _askatana_layer_anim_bp (Utilities|Casting|CastToKatanaLayerAnimBP (Components|SkeletalMesh|AnimationBlueprintLinking|GetLinkedAnimLayerInstancebyClass (Variables|/Script/Engine.Character|GetMesh _fortplayerpawn_2) "/KatanaGameplay/Items/Katana/Animation/KatanaLayerAnimBP.KatanaLayerAnimBP_C")))
              (Class|KatanaLayerAnimBP|SetDashChargePlayRate _askatana_layer_anim_bp (Math|Float|SafeDivide 1.0 (Utilities|RemapFloat (Math|Curves|GetFloatValue _chargeslomocurve _output) "(X=0.000000,Y=1.000000)" (Math|Vector2D|MakeVector2D _returnvalue_6 1.0) true))))
            (:else)))
        (:else)))))

(event Custom|LaunchPlayerUp
  (params (ZVelocity :type Float%20%28double-precision%29))
  ;[1 1 1] Launch player up after entering falling movement mode. Then end ability to continue the chargeup.
  (bind _fortplayerpawn (Variables|Default|GetFortPlayerPawn)
    (:then
      (Networking|ForceNetUpdate _fortplayerpawn)
      (Character|LaunchCharacter _fortplayerpawn (Math|Vector|MakeVector (.x (Variables|Dash|Eject|GetCachedEjectVelocity)) (.y (Variables|Dash|Eject|GetCachedEjectVelocity)) ZVelocity) false true)
      (Utilities|FlowControl|DelayUntilNextTick)
      (bind _fortplayerpawn_2 (Variables|Default|GetFortPlayerPawn)
        (:then
          (Networking|ForceNetUpdate _fortplayerpawn_2)
          (if (GameplayTags|HasMatchingGameplayTag (Variables|Default|GetFortPlayerPawn) "(TagName=\"MovementMode.SurfaceSwimming\")")
            (CallFunction|LaunchPlayerUp (ScalableFloat|GetValueatLevel (Variables|Dash|Eject|GetJumpOutOfWaterZVelocity)))
            (else
              (CallFunction|StopDashChargeEject))))
        (:else)))
    (:else)))

(event Mantis|EventOnMantisTechniqueHit
  (params (TargetDataHandle :type Gameplay%20Ability%20Target%20Data%20Handle%20Structure%20%28by%20ref%29) (ApplicationTag :type Gameplay%20Tag%20Structure))
  (bind _fortplayerpawn (Variables|Default|GetFortPlayerPawn))
  (bind _returnvalue (Weapon|GetCurrentWeapon _fortplayerpawn))
  ;[1 0.2238 0.8693] Ignore pawn collisions when dashing. Do this here instead of in the DashAttack GA for two reasons:
  ; - we already did a trace to deal damage to the players in the path of the dash, we can take advantage of that using the target data of the dash attack.
  ; - because the targeting for the damage and the application of the DashAttack GA happen at roughly the same time, the dash movement occasionally occurs before the targeting, resulting in the targeting failing because all pawns that would be dealt damage have their collision ignored.
  (|CallParentFunction TargetDataHandle ApplicationTag)
  (if (== ApplicationTag "(TagName=\"Mantis.Katana.Melee.DashAttack.Damage\")")
    (if (ScalableFloat|GetValueasBool (Variables|Default|GetEnableIgnoreCollision))
      (Variables|Dash|Attack|SetDashAttackTargets TargetDataHandle)
      (for _array_element (Ability|TargetData|GetAllActorsfromTargetData TargetDataHandle)
        (Utilities|IsValid _array_element
          (:Is%20Valid
            (if (ScalableFloat|GetValueasBool (Variables|Default|GetEnableIgnoreCollisionDestroyCheck))
              (bind _asbuilding_actor (Utilities|Casting|CastToBuildingActor _array_element)
                (:then
                  (bind _bwilldashdestroybuilding (CallFunction|WillDashDestroyBuilding _asbuilding_actor))
                  (if _bwilldashdestroybuilding
                    _))
                (:CastFailed
                  _))
              (Collision|IgnoreActorwhenMoving (Variables|/Script/Engine.Character|GetCapsuleComponent (Variables|Default|GetFortPlayerPawn)) _array_element true)
              (Utilities|Array|Add (Variables|Dash|Attack|GetMoveIgnoredActors) _array_element)
              (else
                _))))))
    (elif (GameplayTags|MatchesTag ApplicationTag "(TagName=\"Mantis.Katana.Melee.PrimaryAttack.Damage\")")
      (Utilities|FlowControl|DoOnce false
        (:Completed
          (if (Utilities|IsA _returnvalue "/KatanaGameplay/Items/Katana/B_Melee_Katana.B_Melee_Katana_C")
            (FortWeapon|UseWeaponDurability(Scaled) _returnvalue (ScalableFloat|GetValueatLevel (Variables|Default|GetHF_PrimaryDurabilityCost)))))))))

(event Custom|ForceDash
  ;[1 0.2487 0.8122] Dash Charge
  ;;[1 0.09327 0.8085] Force Dash
  (Ability|HasAuthority
    (:False
      (bind _fortplayerpawn (Variables|Default|GetFortPlayerPawn)
        (:then
          (Networking|ForceNetUpdate _fortplayerpawn)
          (Utilities|IsValid (Actor|GetComponentbyClass _fortplayerpawn "/Script/MantisRuntime.FortMantisPawnComponent")
            (:Is%20Valid
              (bind _returnvalue (Mantis|TryManuallyStartTechnique (Actor|GetComponentbyClass _fortplayerpawn "/Script/MantisRuntime.FortMantisPawnComponent") 2))
              (if (not _returnvalue)
                (Mantis|TryManuallyEndCurrentTechnique (Actor|GetComponentbyClass _fortplayerpawn "/Script/MantisRuntime.FortMantisPawnComponent"))))))
        (:else)))))

(event Custom|ResetDurabilityCost
  (bind _fortplayerpawn (Variables|Default|GetFortPlayerPawn))
  (bind _returnvalue (Weapon|GetCurrentWeapon _fortplayerpawn))
  ;[1 0.2238 0.8693] Ignore pawn collisions when dashing. Do this here instead of in the DashAttack GA for two reasons:
  ; - we already did a trace to deal damage to the players in the path of the dash, we can take advantage of that using the target data of the dash attack.
  ; - because the targeting for the damage and the application of the DashAttack GA happen at roughly the same time, the dash movement occasionally occurs before the targeting, resulting in the targeting failing because all pawns that would be dealt damage have their collision ignored.
  (Utilities|FlowControl|DoOnce false
    (:Completed
      (if (Utilities|IsA _returnvalue "/KatanaGameplay/Items/Katana/B_Melee_Katana.B_Melee_Katana_C")
        (FortWeapon|UseWeaponDurability(Scaled) _returnvalue (ScalableFloat|GetValueatLevel (Variables|Default|GetHF_PrimaryDurabilityCost)))))))

(event Custom|ClearMoveIgnoredActors
  (bind _moveignoredactors (Variables|Dash|Attack|GetMoveIgnoredActors))
  ;[1 0.2238 0.8693] Ignore pawn collisions when dashing. Do this here instead of in the DashAttack GA for two reasons:
  ; - we already did a trace to deal damage to the players in the path of the dash, we can take advantage of that using the target data of the dash attack.
  ; - because the targeting for the damage and the application of the DashAttack GA happen at roughly the same time, the dash movement occasionally occurs before the targeting, resulting in the targeting failing because all pawns that would be dealt damage have their collision ignored.
  (for _array_element _moveignoredactors
    (Utilities|IsValid _array_element
      (:Is%20Valid
        (Collision|IgnoreActorwhenMoving (Variables|/Script/Engine.Character|GetCapsuleComponent (Variables|Default|GetFortPlayerPawn)) _array_element false))))
  (Utilities|Array|Clear _moveignoredactors))

(event Custom|StopDashAttackRootMotion
  ;[0.08689 0.4707 1] Dash Attack
  (Utilities|FlowControl|Sequence
    (:then_0
      (Utilities|Time|ClearandInvalidateTimerbyHandle (Variables|Dash|Attack|GetEndDashAttackTimerHandle))
      (bind _fortplayerpawn (Variables|Default|GetFortPlayerPawn)
        (:then
          (Networking|ForceNetUpdate _fortplayerpawn)
          (Variables|/Script/Engine.CharacterMovementComponent|SetGravityScale 1.0 (Variables|/Script/Engine.Character|GetCharacterMovement _fortplayerpawn)) ; Reset gravity scale. Originally set\nin DashCharge
          (CallFunction|ClearMoveIgnoredActors)
          (if (GameplayTags|HasAnyMatchingGameplayTags _fortplayerpawn (Variables|Default|GetBlockGrindRailsTag))
            (Ability|GameplayEffect|RemoveLooseGameplayTags (Variables|Default|GetFortPlayerPawn) (Variables|Default|GetBlockGrindRailsTag))))
        (:else)))
    (:then_1
      (bind _dashattackrootmotiontask (Variables|Dash|Attack|GetDashAttackRootMotionTask)
        (:then
          (GameplayTasks|EndTask _dashattackrootmotiontask)
          (Variables|Dash|Attack|SetDashAttackRootMotionTask 0)
          (bind _fortplayerpawn_2 (Variables|Default|GetFortPlayerPawn)
            (:then
              (Variables|/Script/Engine.MovementComponent|SetVelocity (Math|Vector|SelectVector "0, 0, 0" (* (Math|Vector|GetForwardVector (Variables|Dash|Attack|GetDashDirection)) (ScalableFloat|GetValueatLevel (Variables|Dash|Attack|GetSpeedMaintanedAfterDash))) (Variables|Dash|Attack|GetbShouldStopAfterDash)) (Variables|/Script/Engine.Character|GetCharacterMovement _fortplayerpawn_2))
              (CallFunction|ApplyFallDamageImmunity (Variables|Default|GetGE_StopDash_FallDamageImmunity))
              (bind _returnvalue (Utilities|Time|SetTimerbyEvent (EventDispatchers|CreateEvent :!event Custom|ManuallyEndDashAttack) 0.8))
              (Variables|Dash|Attack|SetEndDashAttackTimerHandle _returnvalue)
              (Camera|PopCameraMode)
              (CallFunction|SendDashQuestEvent))
            (:else)))
        (:else)))))

(event Custom|StopDashChargeEject
  (bind _fortplayerpawn (Variables|Default|GetFortPlayerPawn))
  (bind _returnvalue (Actor|GetComponentbyClass _fortplayerpawn "/Script/MantisRuntime.FortMantisPawnComponent"))
  (bind _fortplayerpawn_2 (Variables|Default|GetFortPlayerPawn)
    (:then
      (Character|UnCrouch _fortplayerpawn_2)
      (Ability|HasAuthority
        (:False
          (Ability|GameplayEffect|AddLooseGameplayTags _fortplayerpawn (Variables|Dash|Charge|GetBlockEquipAnimTag))
          (Utilities|IsValid _returnvalue
            (:Is%20Valid
              (Mantis|SetBranchRule _returnvalue "Any" "Default" false)
              (bind _returnvalue_2 (Mantis|TryManuallyStartTechnique _returnvalue 1))
              (if (not _returnvalue_2)
                (Mantis|TryManuallyEndCurrentTechnique _returnvalue)))))))
    (:else)))

(event Custom|SendDashQuestEvent
  (bind (_returnvalue_x _returnvalue_y) (Math|Vector|BreakVector (* (Math|Vector|GetForwardVector (Variables|Dash|Attack|GetDashDirection)) (Variables|Dash|Attack|GetActualDashDistance))))
  ;[1 0.2285 0.9415] Send a quest event with the 2D dash distance as the magnitude.
  (bind _fortplayerpawn (Variables|Default|GetFortPlayerPawn)
    (:then
      (Ability|SendGameplayEventtoActor _fortplayerpawn "(TagName=\"Athena.Event.Gameplay.Katana.DashDistance\")" (Utilities|Struct|MakeGameplayEventData "(TagName=\"Athena.Event.Gameplay.Katana.DashDistance\")" 0 0 0 0 "()" "(GameplayTags=)" "(GameplayTags=)" (Math|Vector|VectorLength (Math|Vector|MakeVector _returnvalue_x _returnvalue_y)))))
    (:else)))

(event Custom|StopDashCharge
  ;[1 0.2487 0.8122] Dash Charge
  ;;[1 0.5873 0.1156] StopDashCharge
  (Utilities|FlowControl|Sequence
    (:then_0
      (bind _chargecameramodifier (Variables|Dash|Charge|GetChargeCameraModifier)
        (:then
          (CameraModifier|DisableModifier _chargecameramodifier false))
        (:else)))
    (:then_1
      (bind _fortplayerpawn (Variables|Default|GetFortPlayerPawn)
        (:then
          (Networking|ForceNetUpdate _fortplayerpawn)
          (bind _asb_melee_katana (Utilities|Casting|CastToB_Melee_Katana (Weapon|GetCurrentWeapon _fortplayerpawn)))
          (Class|BMeleeKatana|SetDashChargeEnabled _asb_melee_katana false))
        (:else)))
    (:then_2
      (Utilities|Time|ClearandInvalidateTimerbyHandle (Variables|Dash|Charge|GetUpdateSlomoTimerHandle))
      (bind _fortplayerpawn_2 (Variables|Default|GetFortPlayerPawn)
        (:then
          (Variables|/Script/Engine.Actor|SetCustomTimeDilation 1.0 _fortplayerpawn_2))
        (:else)))))

(event Custom|ConsumeAmmo
  ;[0.08689 0.4707 1] Dash Attack
  ;;[0.2181 0.1801 1] Consume Ammo - skip if the ignore ammo cheat is being used.
  (Ability|HasAuthority
    (:True
      (bind _fortplayerpawn (Variables|Default|GetFortPlayerPawn)
        (:then
          (bind _asb_melee_katana (Utilities|Casting|CastToB_Melee_Katana (Variables|/Script/FortniteGame.FortPawn|GetCurrentWeapon _fortplayerpawn)))
          (if (not (and (Variables|/KatanaGameplay/Items/Katana/B_Melee_Katana.B_Melee_Katana_C|GetbDebugIgnoreAmmo _asb_melee_katana) (Editor|InEditororPIE)))
            (FortWeapon|Ammo|ConsumeAmmo _asb_melee_katana 1 false)))
        (:else)))))

(event Custom|ManuallyEndDashAttack
  ;[0.08689 0.4707 1] Dash Attack
  (bind _fortplayerpawn (Variables|Default|GetFortPlayerPawn)
    (:then
      (Networking|ForceNetUpdate _fortplayerpawn)
      (Utilities|IsValid (Actor|GetComponentbyClass _fortplayerpawn "/Script/MantisRuntime.FortMantisPawnComponent")
        (:Is%20Valid
          (Mantis|TryManuallyEndCurrentTechnique (Actor|GetComponentbyClass _fortplayerpawn "/Script/MantisRuntime.FortMantisPawnComponent")))))
    (:else)))

(fn AddDashAttackDynamicEffectContainer
  (bind _originaltargetselection_list (Utilities|Struct|BreakFortAbilityTargetSelectionList (Variables|GetOriginalTargetSelection_List)))
  (bind (_originaltargetselection_list_2 _originaltargetselection_bstopatfirstsuccess _originaltargetselection_bkeepcheckinglistonindestructiblehit _originaltargetselection_abilitytargetblockedgameplaytags _originaltargetselection_buseweaponranges _originaltargetselection_busemaxyawangletotarget _originaltargetselection_maxyawangletotarget) (Utilities|Struct|BreakFortAbilityTargetSelectionList (Variables|GetOriginalTargetSelection_List)))
  (bind _fortplayerpawn (Variables|Default|GetFortPlayerPawn)
    (:then
      (Utilities|IsValid (Actor|GetComponentbyClass _fortplayerpawn "/Script/MantisRuntime.FortMantisPawnComponent")
        (:Is%20Valid
          (bind (_returnvalue_intvalues _returnvalue_floatvalues) (Utilities|Struct|BreakFortMantisTechniqueMetadata (Mantis|GetCurrentTechniqueMetadata (Actor|GetComponentbyClass _fortplayerpawn "/Script/MantisRuntime.FortMantisPawnComponent"))))
          (bind _output_get (LocalVariables|SetDashDistance (Utilities|Map|Find _returnvalue_floatvalues "(TagName=\"Mantis.Metadata.Katana.DashDistance\")")))
          (LocalVariables|SetHalfDashDistance (/ (+ (ScalableFloat|GetValueatLevel (Variables|Dash|Attack|GetDashAdditionalForwardAttackRange)) _output_get) 2.0))
          (Utilities|Array|Clear (LocalVariables|GetModifiedTargetSelectionArray))
          (bind _returnvalue (Ability|GetGameplayEffectContainer (Variables|Dash|Attack|GetDashAttackEffectContainerApplicationTag)))
          (bind (_returnvalue_applicationtag _returnvalue_targetselection _returnvalue_targetgameplayeffectclasses _returnvalue_ownergameplayeffectclasses _returnvalue_activationcues _returnvalue_impactcues _returnvalue_busecalculationintooltips _returnvalue_boverridechargemagnitude _returnvalue_chargemagnitudeoverridevalue) (Utilities|Struct|BreakFortGameplayEffectContainer _returnvalue))
          (LocalVariables|SetOriginalTargetSelection _returnvalue_targetselection)
          (Utilities|FlowControl|Sequence
            (:then_0
              (for _array_element _originaltargetselection_list
                (Utilities|Array|Add (LocalVariables|GetModifiedTargetSelectionArray) (Utilities|Struct|MakeFortAbilityTargetSelection :Shape (Utilities|Struct|BreakFortAbilityTargetSelection _array_element) :CustomShapeComponentName (.CustomShapeComponentName (Utilities|Struct|BreakFortAbilityTargetSelection _array_element)) :TestType (.TestType (Utilities|Struct|BreakFortAbilityTargetSelection _array_element)) :PrimarySource (.PrimarySource (Utilities|Struct|BreakFortAbilityTargetSelection _array_element)) :SecondarySource (.SecondarySource (Utilities|Struct|BreakFortAbilityTargetSelection _array_element)) :Range (.Range (Utilities|Struct|BreakFortAbilityTargetSelection _array_element)) :HalfExtents (Math|Vector|MakeVector (LocalVariables|GetHalfDashDistance) (ScalableFloat|GetValueatLevel (Variables|Dash|Attack|GetDashDamageTraceHalfWidth)) (ScalableFloat|GetValueatLevel (Variables|Dash|Attack|GetDashDamageTraceHalfHeight))) :bAlignShapeEdgeToSourceLocation (.bAlignShapeEdgeToSourceLocation (Utilities|Struct|BreakFortAbilityTargetSelection _array_element)) :ConeYawAngle (.ConeYawAngle (Utilities|Struct|BreakFortAbilityTargetSelection _array_element)) :ConePitchAngle (.ConePitchAngle (Utilities|Struct|BreakFortAbilityTargetSelection _array_element)) :ConeMinRadius (.ConeMinRadius (Utilities|Struct|BreakFortAbilityTargetSelection _array_element)) :TargetFilter (.TargetFilter (Utilities|Struct|BreakFortAbilityTargetSelection _array_element)) :bExcludeObstructedByWorld (.bExcludeObstructedByWorld (Utilities|Struct|BreakFortAbilityTargetSelection _array_element)) :bShouldAttachedActorsObstructTarget (.bShouldAttachedActorsObstructTarget (Utilities|Struct|BreakFortAbilityTargetSelection _array_element)) :bCreateHitResultWhenNoTargetsFound (.bCreateHitResultWhenNoTargetsFound (Utilities|Struct|BreakFortAbilityTargetSelection _array_element)) :bUseProjectileRotationForDamageZones (.bUseProjectileRotationForDamageZones (Utilities|Struct|BreakFortAbilityTargetSelection _array_element)) :TargetSelectionUsage (.TargetSelectionUsage (Utilities|Struct|BreakFortAbilityTargetSelection _array_element)) :MaxTargets (.MaxTargets (Utilities|Struct|BreakFortAbilityTargetSelection _array_element))))))
            (:then_1
              (Ability|AddDynamicGameplayEffectContainer 0 (Variables|Dash|Attack|GetDashAttackEffectContainerApplicationTag) (Utilities|Struct|MakeFortGameplayEffectContainer (Variables|Dash|Attack|GetDashAttackEffectContainerApplicationTag) (Utilities|Struct|MakeFortAbilityTargetSelectionList (LocalVariables|GetModifiedTargetSelectionArray) _originaltargetselection_bstopatfirstsuccess _originaltargetselection_bkeepcheckinglistonindestructiblehit _originaltargetselection_buseweaponranges _originaltargetselection_busemaxyawangletotarget _originaltargetselection_maxyawangletotarget) _returnvalue_targetgameplayeffectclasses _returnvalue_ownergameplayeffectclasses _returnvalue_activationcues _returnvalue_impactcues _returnvalue_busecalculationintooltips _returnvalue_boverridechargemagnitude _returnvalue_chargemagnitudeoverridevalue)))))))
    (:else)))

(fn ApplyFallDamageImmunity
  (params (GameplayEffectClass :type Gameplay%20Effect%20Class%20Reference))
  ;[0.2852 1 0.7772] Apply fall damage immunity if player is falling
  (bind _fortplayerpawn (Variables|Default|GetFortPlayerPawn)
    (:then
      (if (GameplayTags|HasMatchingGameplayTag _fortplayerpawn "(TagName=\"MovementMode.Falling\")")
        (Ability|ApplyGameplayEffectToOwner GameplayEffectClass)))
    (:else)))

(fn ConfigureRootMotionWarping
  (params (TechniqueName :type Name :default "None"))
  (bind _fortplayerpawn (Variables|Default|GetFortPlayerPawn)
    (:then
      (Utilities|IsValid (Weapon|GetCurrentWeapon _fortplayerpawn)
        (:Is%20Valid
          (Utilities|IsValid (Actor|GetComponentbyClass (Weapon|GetCurrentWeapon _fortplayerpawn) "/Script/KatanaGameplayRuntime.KatanaTargetingComponent")
            (:Is%20Valid
              (LocalVariables|SetTargetingComponent (Actor|GetComponentbyClass (Weapon|GetCurrentWeapon _fortplayerpawn) "/Script/KatanaGameplayRuntime.KatanaTargetingComponent"))
              (switch name TechniqueName
                (:Default
                  (return false false 0 false false "0, 0, 0" false 0))
                (:PrimaryAttack_1
                  (bind _returnvalue (Katana|PrimaryAttack|PerformPrimaryAttackTargeting (LocalVariables|GetTargetingComponent) "OnGroundFirst"))
                  (return true true 0 false true _returnvalue false 0))
                (:PrimaryAttack_2
                  (bind _returnvalue_2 (Katana|PrimaryAttack|PerformPrimaryAttackTargeting (LocalVariables|GetTargetingComponent) "OnGroundSecond"))
                  (return true true 0 false true _returnvalue_2 false 0))
                (:InAirAttack_Slam
                  (bind _returnvalue_3 (Katana|PrimaryAttack|PerformPrimaryAttackTargeting (LocalVariables|GetTargetingComponent) "InAir"))
                  (return true true 0 false true _returnvalue_3 false 0))))))))
    (:else)))

(fn FillMetadataForTechnique
  (params (TechniqueName :type Name :default "None"))
  (switch name TechniqueName
    (:Default
      (return 0))
    (:DashAttack
      ;[1 0.1944 0.7352] Dash Targeting (V2
      (bind _fortplayerpawn (Variables|Default|GetFortPlayerPawn)
        (:then
          (Utilities|IsValid (Weapon|GetCurrentWeapon _fortplayerpawn)
            (:Is%20Valid
              (Utilities|IsValid (Actor|GetComponentbyClass (Weapon|GetCurrentWeapon _fortplayerpawn) "/Script/KatanaGameplayRuntime.KatanaTargetingComponent")
                (:Is%20Valid
                  (bind (_outdashtargetinginfo _returnvalue) (Katana|Dash|PerformDashTargeting (Actor|GetComponentbyClass (Weapon|GetCurrentWeapon _fortplayerpawn) "/Script/KatanaGameplayRuntime.KatanaTargetingComponent")))
                  (if _returnvalue
                    (return (Utilities|Map|MakeMap "(TagName=\"Mantis.Metadata.Katana.ShouldStopAfterDash\")" (Math|Conversions|ToInteger(Boolean) (Utilities|IsValid (.DashBlockingActor (Utilities|Struct|BreakFortKatanaDashTargetingInfo _outdashtargetinginfo)) :!class /Script/Engine.KismetSystemLibrary))) (Utilities|Map|MakeMap "(TagName=\"Mantis.Metadata.Katana.DashInputHeldDuration\")" (Variables|Default|GetCurrentTechniqueInputHeldDuration) "(TagName=\"Mantis.Metadata.Katana.DashDistance\")" (.DashDistance (Utilities|Struct|BreakFortKatanaDashTargetingInfo _outdashtargetinginfo))) (Utilities|Map|MakeMap "(TagName=\"Mantis.Metadata.Katana.DashDirection\")" (Utilities|Struct|BreakFortKatanaDashTargetingInfo _outdashtargetinginfo) "(TagName=\"Mantis.Metadata.Katana.DashOrigin\")" (Transformation|GetActorLocation _fortplayerpawn)) 0)))))))
        (:else)))))

(fn GetCustomAbilitySourceTransform
  (Utilities|FlowControl|Sequence
    (:then_0
      (bind _fortplayerpawn (Variables|Default|GetFortPlayerPawn)
        (:then
          (Utilities|IsValid (Actor|GetComponentbyClass _fortplayerpawn "/Script/MantisRuntime.FortMantisPawnComponent")
            (:Is%20Valid
              (if (Variables|Default|GetbUseFinalSlashTargetingTransform)
                (bind (_returnvalue_intvalues _returnvalue_floatvalues _returnvalue_vectorvalues) (Utilities|Struct|BreakFortMantisTechniqueMetadata (Mantis|GetCurrentTechniqueMetadata (Actor|GetComponentbyClass _fortplayerpawn "/Script/MantisRuntime.FortMantisPawnComponent"))))
                (bind (_returnvalue_location _returnvalue_rotation _returnvalue_scale) (Math|Transform|BreakTransform (Ability|GetTargetingTransform _fortplayerpawn "PawnTowardsFocus" 0)))
                (return _returnvalue_location (Math|Conversions|RotationFromXVector (Utilities|Map|Find _returnvalue_vectorvalues "(TagName=\"Mantis.Metadata.Katana.DashDirection\")")) _returnvalue_scale)
                (else
                  (return (Utilities|Map|Find _returnvalue_vectorvalues "(TagName=\"Mantis.Metadata.Katana.DashOrigin\")") (Math|Conversions|RotationFromXVector (Utilities|Map|Find _returnvalue_vectorvalues "(TagName=\"Mantis.Metadata.Katana.DashDirection\")")) "1.000000,1.000000,1.000000"))))))
        (:else)))
    (:then_1
      (return 0))))

(fn InitAbility
  (bind _output_get (|CallParentFunction))
  (return _output_get))

(fn ValidateDashMetadata
  (bind _maxdashdistance (Variables|Dash|Attack|GetMaxDashDistance))
  (bind _returnvalue (ScalableFloat|GetValueatLevel _maxdashdistance))
  (bind _dashadditionalforwardattackrange (Variables|Dash|Attack|GetDashAdditionalForwardAttackRange))
  (bind _returnvalue_2 (ScalableFloat|GetValueatLevel _dashadditionalforwardattackrange))
  (bind _fortplayerpawn (Variables|Default|GetFortPlayerPawn))
  (bind _returnvalue_3 (Transformation|GetActorLocation _fortplayerpawn))
  (bind _returnvalue_4 (Utilities|Operators|Add _returnvalue _returnvalue_2 2e+02))
  (Utilities|FlowControl|Sequence
    (:then_0
      (if (> (Variables|Dash|Attack|GetActualDashDistance) _returnvalue_4)
        (Variables|Dash|Attack|SetActualDashDistance _returnvalue_4)
        (Development|PrintString "Dash validation failed: distance" false)))
    (:then_1
      (Utilities|IsValid _fortplayerpawn
        (:Is%20Valid
          (if (> (Math|Vector|Distance(Vector) (Variables|Dash|Attack|GetDashOrigin) _returnvalue_3) 2e+02)
            (Variables|Dash|Attack|SetDashOrigin _returnvalue_3)
            (Development|PrintString "Dash validation failed: origin" false)))))))

(fn WillDamageAmountDestroyBuilding
  (params (Building :type Building%20Actor%20Object%20Reference) (Damage :type Float%20%28double-precision%29))
  (Utilities|IsValid Building
    (:Is%20Valid
      (return (<= (Math|Float|SelectFloat (Health|GetHealth Building :!class /Script/FortniteGame.BuildingActor) (- (Health|GetHealth Building :!class /Script/FortniteGame.BuildingActor) Damage) (Networking|HasAuthority Building)) 0)))
    (:Is%20Not%20Valid
      (return false))))

(fn WillDashDestroyBuilding
  (params (Building :type Building%20Actor%20Object%20Reference))
  (Utilities|FlowControl|Sequence
    (:then_0
      (Utilities|IsValid Building
        (:Is%20Valid
          (bind _asbuilding_static_mesh (Utilities|Casting|CastToBuildingSMActor Building)
            (:then
              (LocalVariables|SetbWillBeDestroyed (and (Class|BuildingStaticMesh|CanbeDamaged _asbuilding_static_mesh (Pawn|GetController (Variables|Default|GetFortPlayerPawn)) (Variables|Default|GetFortPlayerPawn)) (CallFunction|WillDamageAmountDestroyBuilding (Variables|GetBuilding) (ScalableFloat|GetValueatLevel (LocalVariables|GetBuildingRemainingHealthToDestroy))))))
            (:CastFailed
              (LocalVariables|SetbWillBeDestroyed (and (Variables|/Script/Engine.Actor|GetbCanBeDamaged Building) (CallFunction|WillDamageAmountDestroyBuilding Building (ScalableFloat|GetValueatLevel (LocalVariables|GetPropRemainingHealthToDestroy))))))) ; If this building is a BuildingSMActor, CanBeDamaged is \nmore accurate than bCanBeDamaged
        )))
    (:then_1
      (return (LocalVariables|GetbWillBeDestroyed)))))
